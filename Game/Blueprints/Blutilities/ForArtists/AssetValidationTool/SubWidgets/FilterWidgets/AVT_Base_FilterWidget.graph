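# Base Editor Utility Widget for Asset Validation Tool filters, with no widgets of its own.
# runFilterOnMesh calls filterAsset on valid assets when enabled; that check and both validate functions are stubs, apparently for child overrides.
# pickYourFilterData selects saved data by filterID, and setFilterData applies the default enabled state on first use.
# Data and enabled-state changes fire the onSaveGameTriggered dispatcher, then call validateWidgets.
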EditorUtilityWidgetBlueprint /Game/Blueprints/Blutilities/ForArtists/AssetValidationTool/SubWidgets/FilterWidgets/AVT_Base_FilterWidget
graphs [
 EventGraph {nodes [
   Event Construct()
   > Set currentFilterData(currentFilterData:AVT Filter Data Struct Structure=Make AVT Filter Data Struct!(
     filterID_13_#1:Name=Get filterID() -> (filterID:Name=None)
     filterEnabled_15_#2:bool=True
     hasHadDataSet_17_#3:bool=False
    ) -> (AVT_FilterData_Struct:AVT Filter Data Struct Structure) {is=title K2Node_MakeStruct}) -> (Output_Get:AVT Filter Data Struct Structure)
  ]}
 filterAsset {nodes [
   filterAsset!() -> (then, asset:Object Reference) {is=title K2Node_FunctionEntry}
   > Return Node!(execute, hasError:bool=false, ErrorStruct:AVT Error Struct Structure)
  ]}
 getFilterData {nodes [
   getFilterData() {is=title K2Node_FunctionEntry}
   > validateFilterData!(execute, self) -> (then, " ":bool)
   > Set members in AVT Filter Data Struct!(
    execute:Exec
    StructRef:"AVT Filter Data Struct Structure (by ref)"=Get currentFilterData() -> (currentFilterData:AVT Filter Data Struct Structure)
    filterID_13_#1:Name=Get filterID() -> (filterID=None)
   ) -> (then:Exec, StructOut:"AVT Filter Data Struct Structure (by ref)") {is=title K2Node_SetFieldsInStruct}
   > Return Node!(execute, data:AVT Filter Data Struct Structure=Get currentFilterData())
  ]}
 getFilterEnabled {nodes [
   getFilterEnabled() {is=title K2Node_FunctionEntry}
   > validateFilterData()
   > Return Node!(
    execute
    isEnabled:bool=Break AVT Filter Data Struct!(AVT_FilterData_Struct:"AVT Filter Data Struct Structure (by ref)"=Get currentFilterData()) -> (filterEnabled_15_#2:bool=~) {is=title K2Node_BreakStruct}
   )
  ]}
 lerpErrorColoring {nodes [
   n1 = lerpErrorColoring!() -> (then, Alpha:double=~) {is=title K2Node_FunctionEntry}
   > Return Node!(execute, color:Linear Color Structure="Lerp (LinearColor)"(A=(R=0.1651,G=0.5149,B=0.0203,A=1), B=(R=0.3968,G=0,B=0,A=1), Alpha=@n1))
  ]}
 onSaveGameTriggered {nodes [onSaveGameTriggered() {is=title K2Node_FunctionEntry}]}
 pickYourFilterData {nodes [
   n1 = pickYourFilterData!() -> (then, data:Array of AVT Filter Data Struct Structures) {is=title K2Node_FunctionEntry}
   > n2 = For Each Loop with Break!(Exec:Exec, Array:Array of AVT Filter Data Struct Structures=@n1, Break:Exec=@n3.then) -> (LoopBody:Exec, Array Element:AVT Filter Data Struct Structure, Array Index:int, Completed:Exec)
   Return Node!(execute=@n2.Completed, chosenIndex:int=@n2.Array Index)
   > Branch(Condition="Equal (Name)"!(
     A:Name=Break AVT Filter Data Struct!(AVT_FilterData_Struct=@n2) -> (
      Bools_2_#4:Array of Booleans
      Strings_5_#5:Array of Strings
      Floats_8_#6:"Array of Float (single-precision)s"
      filterID_13_#1:Name
      filterEnabled_15_#2=~
      hasHadDataSet_17_#3:bool=~
     ).filterID_13_#1
     B:Name=Get filterID() -> (filterID=None)
    ) -> (ReturnValue:bool) {is=title K2Node_CallFunction})
   > n3 = setFilterData!(execute, self, data:AVT Filter Data Struct Structure=@n2) -> (then)
  ]}
 runFilterOnMesh {nodes [
   n1 = runFilterOnMesh!() -> (then, asset:Object Reference) {is=title K2Node_FunctionEntry}
   > n2 = Is Valid(InputObject=@n1)
   > n3 = Branch(Condition=getFilterEnabled!() -> (isEnabled:bool) {is=title K2Node_CallFunction})
   > n4 = filterAsset!(execute, self, asset:Object Reference=@n1) -> (then, hasError:bool, ErrorStruct:AVT Error Struct Structure) {is=title K2Node_CallFunction}
   > Return Node!(execute, hasError:bool=@n4, Error_Struct:AVT Error Struct Structure=@n4.ErrorStruct)
   Return Node(execute=[@n3.else, @n2.Is Not Valid], hasError=false)
  ]}
 setFilterData {nodes [
   n1 = setFilterData!() -> (then, data:AVT Filter Data Struct Structure) {is=title K2Node_FunctionEntry}
   > n2 = Set currentFilterData(currentFilterData=@n1)
   > n3 = Branch(Condition=Break AVT Filter Data Struct!(AVT_FilterData_Struct=@n2) -> (hasHadDataSet_17_#3=~))
   Set members in AVT Filter Data Struct(
    execute=[@n3.then, @n4.then]
    StructRef=Get currentFilterData()
    filterID_13_#1=Get filterID() -> (filterID=None)
   )
   > triggerSaveGame()
   > validateWidgets!(execute, self) -> (then, " ":bool)
   n4 = Set members in AVT Filter Data Struct!(
    execute=@n3.else
    StructRef=Get currentFilterData()
    filterEnabled_15_#2:bool=Get isFilteredEnabledByDefault() -> (isFilteredEnabledByDefault:bool)
    hasHadDataSet_17_#3:bool=true
   ) -> (then, StructOut)
  ]}
 setFilterEnabled {nodes [
   n1 = setFilterEnabled!() -> (then, isEnabled:bool=~) {is=title K2Node_FunctionEntry}
   > n2 = Branch(Condition="Not Equal (Boolean)"!(A:bool=getFilterEnabled(), B:bool=@n1) -> (ReturnValue:bool) {is=title K2Node_CallFunction})
   > Set members in AVT Filter Data Struct!(execute, StructRef=Get currentFilterData(), filterEnabled_15_#2=@n1) -> (then, StructOut)
   > triggerSaveGame()
   > validateWidgets()
   > Return Node!(execute, changed:bool=true)
   Return Node(execute=@n2.else, changed=false)
  ]}
 triggerSaveGame {nodes [
   triggerSaveGame() {is=title K2Node_FunctionEntry}
   > Call On Save Game Triggered!(execute:Exec, self:AVT Base Filter Widget Object Reference) -> (then:Exec) {is=title K2Node_CallDelegate}
  ]}
 validateFilterData {nodes [
   validateFilterData() {is=title K2Node_FunctionEntry}
   > Return Node!(execute, " ":bool=false)
  ]}
 validateWidgets {nodes [
   validateWidgets() {is=title K2Node_FunctionEntry}
   > Return Node(" "=false)
  ]}
]
variables [
 currentFilterData {type=AVT_FilterData_Struct, default {filterEnabled=true, filterId=None, hasHadDataSet=false, bools [], floats [], strings []}}
 filterID {type=string, default=None}
 filterType {type=AVT_FilterCategory_Enum, default=Mesh}
 hasHadDataSet {type=boolean}
 isFilteredEnabledByDefault {type=boolean, default=true}
 onSaveGameTriggered {default=()}
]
widgets {info {rootWidgetClass=null, widgetCount=0}, widgets []}
